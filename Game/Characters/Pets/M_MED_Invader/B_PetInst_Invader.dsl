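; Blueprint for the Invader pet instance, driving its mind-control effects on the player.
; On BeginPlay it binds the mind-control and DBNO Niagara systems to the player's head mesh, or body mesh if that valid head lacks a "head" socket, printing errors otherwise.
; Custom events toggle mind-control and spark effects through Niagara booleans, without disabling the systems.
; The construction script creates body and carrier dynamic materials on the pet mesh.
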(event EventBeginPlay
  (bind _returnvalue (CallFunction|GetFortPlayerPawn))
  (bind _returnvalue_2 (Customization|GetSkeletalMeshforPartType _returnvalue "Body" :!class /Script/FortniteGame.FortPlayerPawn))
  (bind _returnvalue_3 (Customization|GetSkeletalMeshforPartType _returnvalue "Head" :!class /Script/FortniteGame.FortPlayerPawn))
  (|CallParentFunction)
  (Utilities|IsValid _returnvalue_3
    (:Is%20Valid
      (if (Transformation|DoesSocketExist _returnvalue_3 "head")
        ;[0 0 0] Set the player mesh so that the FX Go to the head defined in niagara
        (Niagara|SetNiagaraSkeletalMeshComponent (Variables|Default|GetMindControlFX) "SkeletalMesh1" _returnvalue_3)
        ;[0 0 0] Set the player mesh so that the FX Go to the bone defined in niagara
        (Niagara|SetNiagaraSkeletalMeshComponent (Variables|Default|GetNS_Invader_Pet_MindControl_DBNO) "SkeletalMesh1" _returnvalue_3)
        (else
          (Utilities|IsValid _returnvalue_2
            (:Is%20Valid
              (if (Transformation|DoesSocketExist _returnvalue_2 "head")
                (Niagara|SetNiagaraSkeletalMeshComponent (Variables|Default|GetMindControlFX) "SkeletalMesh1" _returnvalue_2)
                (Niagara|SetNiagaraSkeletalMeshComponent (Variables|Default|GetNS_Invader_Pet_MindControl_DBNO) "SkeletalMesh1" _returnvalue_2)
                (else
                  (Development|PrintString "Head and Body Skeletal Mesh do not have a \"Head\" Bone, send to Tech Anim" true true "(R=1.000000,G=0.000000,B=0.000000,A=1.000000)" 1e+01))))
            (:Is%20Not%20Valid
              (Development|PrintString "Body Skeletal mesh NOT VALID" true true "(R=1.000000,G=0.000000,B=0.000000,A=1.000000)" 1e+01))))))
    (:Is%20Not%20Valid
      (Development|PrintString "Head Skeletal mesh NOT VALID" true true "(R=1.000000,G=0.000000,B=0.000000,A=1.000000)" 1e+01))))

(event Custom|IsMindControllingPlayer
  (params (bIsMindcontrolling :type Boolean))
  ;[0 0 0] show/hide the FX with in the system itself with a timer module, but no disable completely to avoid any potential oddness
  (Effects|Components|ParticleSystem|SetBoolParameter (Variables|Default|GetMindControlFX) "bIsMindcontrolling" bIsMindcontrolling))

(event Custom|SetSParkVFX
  (params (SparkOnParam :type Boolean))
  (Effects|Components|ParticleSystem|SetBoolParameter (Variables|Default|GetNS_Invader_Pet_MindControl_DBNO) "SparkOn" SparkOnParam))

(fn ConstructionScript
  ;[1 1 1] Create MIDs
  (|CallParentFunction)
  (bind _returnvalue (Rendering|Material|CreateDynamicMaterialInstance (Variables|FortPlayerPet|GetPetMesh) 0 0 "None" :!class /Script/Engine.PrimitiveComponent))
  (Variables|Default|SetBodyMID _returnvalue)
  (bind _returnvalue_2 (Rendering|Material|CreateDynamicMaterialInstance (Variables|FortPlayerPet|GetPetMesh) 1 0 "None" :!class /Script/Engine.PrimitiveComponent))
  (Variables|Default|SetCarrierMID _returnvalue_2))
