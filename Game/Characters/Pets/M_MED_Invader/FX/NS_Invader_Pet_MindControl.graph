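# Niagara effect for the Invader pet's mind control, showing a green glowing orb linked to two skeletal meshes, tinted by an HSV shift parameter.
# Mesh rings travel from two hand points on one mesh into the orb, spawning only while the mind-control bool is set, which also gates the orb's light.
# Three twisting, pulsing ribbons stretch from orb to the second mesh's head, their width and orb size driven by a system timer.
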NiagaraSystem /Game/Characters/Pets/M_MED_Invader/FX/NS_Invader_Pet_MindControl
dependencies {
 usedDataInterfaces [&NiagaraDataInterfaceCurve]
 usedDynamicInputs [
  &Boolean_NotOperation
  &Lerp_Vector
  &TransformVector
  &MakeVector
 ]
 usedModules [
  &SystemState
  &FN_Timer
  /Game/Effects/Niagara/Modules/SkeletalMesh/FN_SkeletalBoneData_v2
  &self:SystemSpawnScript.SystemScriptSource.SystemScriptGraph.NiagaraNodeAssignment_1.SetVariables_#1
  &self:SystemSpawnScript.SystemScriptSource.SystemScriptGraph.NiagaraNodeAssignment_2.SetVariables_#2
  &self:SystemSpawnScript.SystemScriptSource.SystemScriptGraph.NiagaraNodeAssignment_0.SetVariables_#3
 ]
}
emitters {
 Hand_Rings {
  emitterUpdateScript [
   &EmitterState()
   &SpawnRate(
    SpawnRate=&MakeCustomFloatFromBool(Bool=$User.bIsMindcontrolling, True Float=5, False Float=0)
    Use Spawn Probability=false
   )
   &SpawnRate(
    SpawnGroup=1
    SpawnRate=&MakeCustomFloatFromBool(Bool=$User.bIsMindcontrolling, True Float=5, False Float=0)
    Use Spawn Probability=false
   ) {module=SpawnRate001}
  ]
  particleSpawnScript [
   &InitializeParticle(
    Color=&HueShiftLinearColor(
     Hue=&MakeFloatFromVector(Vector=$User.HSV Shift, Channel=::NewEnumerator0)
     Saturation=&MakeFloatFromVector(Vector=$User.HSV Shift, Channel=::NewEnumerator1)
     Value=&MakeFloatFromVector(Vector=$User.HSV Shift, Channel=::NewEnumerator2)
     Linear Color=<rgba 0 1 0 0.25>
    )
    Mesh Scale Mode=::NewEnumerator3
    Mesh Uniform Scale=0.045
    Position Offset Coordinate Space=::World
    UsePositionOffset=true
    Write Material Random=true
   )
   &InitialMeshOrientation(Orientation Axis=<0 0 1>, Rotation=<0.25 0 0>, Use Orientation Vector=true, Use Rotation Vector=true)
  ]
  particleUpdateScript [
   &SkeletalMeshLocation(
    Bone / Socket Index=&MaskIntBySpawnGroup(Masked Int=0, Passthrough Int=1, Spawn Group Mask=0)
    Bone Sampling Mode=::NewEnumerator4
    MaskSpawnGroup=false
    Orientation Sampling=::NewEnumerator1
    Position Sampling=::NewEnumerator2
    Skeletal Mesh=$User.SkeletalMesh
   )
   &ParticleState()
   &self:Line_Location_Scratch(
    End Pos:NiagaraPosition=$Particles.SkeletalMeshLocation.SampledPosition
    Pos On Line:NiagaraFloat=&OneMinusFloat(Float=$Particles.NormalizedAge)
    Start Pos:Vector3f=$System.Orb Position
   ) {module=Line_Location_Scratch}
   &ScaleMeshSize(Scale Factor=&VectorFromFloat(Value=&Lerp_Float(A=0.25, B=1, Alpha=$Particles.NormalizedAge)))
   /SetVariables(Particles.Velocity=&Subtract_Vector(A=Linked() {linkedVariable {name=Particles.Position, type=NiagaraPosition}}, B=$System.Orb Position)) {isSetParameters=true}
   &ScaleColor(
    Scale Alpha=&FloatFromCurve(FloatCurve=DataInterface() {propertyValues {Curve {keys [
         {
          arriveTangent=6.8242
          arriveTangentWeight=0.6944
          interpMode=RCIM_Cubic
          leaveTangent=6.8242
          leaveTangentWeight=0.6944
          tangentMode=RCTM_User
          tangentWeightMode=RCTWM_WeightedBoth
          time=0
          value=0
         }
         {
          arriveTangentWeight=0.3883
          interpMode=RCIM_Cubic
          leaveTangentWeight=0.3883
          tangentMode=RCTM_User
          tangentWeightMode=RCTWM_WeightedBoth
          time=0.5
          value=1
         }
         {
          arriveTangent=-6.8262
          arriveTangentWeight=0.6329
          interpMode=RCIM_Cubic
          leaveTangent=-6.8262
          leaveTangentWeight=0.6329
          tangentMode=RCTM_User
          tangentWeightMode=RCTWM_WeightedBoth
          time=1
          value=0
         }
        ]}}})
    ScaleA=true
    ScaleRGB=true
   )
  ]
  rendererClasses [&NiagaraMeshRendererProperties]
  renderers [{rendererClass=&NiagaraMeshRendererProperties, rendererIndex=0}]
 }
 Lines_Orb_To_Head {
  emitterSpawnScript [/SetVariables(Emitter.NumberOfParticlePerRibbon=7, Emitter.NumberOfRibbons=3) {isSetParameters=true}]
  emitterUpdateScript [
   &EmitterState()
   &SpawnBurst_Instantaneous(
    Age:NiagaraFloat=$Emitter.Age
    Spawn Count=&Multiply_Int(A=$Emitter.NumberOfParticlePerRibbon, B=$Emitter.NumberOfRibbons)
    Use Spawn Probability=false
   )
  ]
  particleSpawnScript [
   &InitializeRibbon(
    Color=&HueShiftLinearColor(
     Hue=&MakeFloatFromVector(Vector=$User.HSV Shift, Channel=::NewEnumerator0)
     Saturation=&MakeFloatFromVector(Vector=$User.HSV Shift, Channel=::NewEnumerator1)
     Value=&MakeFloatFromVector(Vector=$User.HSV Shift, Channel=::NewEnumerator2)
     Linear Color=&RandomRangeLinearColor(Minimum=<rgba 0 1.5 0.125 0.8>, Maximum=<rgba 0 1.5 0.25 0.5>)
    )
    Lifetime=2
    Ribbon Width=7.5
    UsePositionOffset=false
    Write Material Random=true
   )
   &FN_SetSpawnGroup(
    Inclusivity=true
    Normalize Spawn Group=true
    Number of Spawn Groups:NiagaraInt32=$Emitter.NumberOfRibbons
    Particles Per Group:NiagaraInt32=$Emitter.NumberOfParticlePerRibbon
    Set Ribbon ID and Ribbon Link Order=true
   )
   /SetVariables(Particles.SpriteSize=<1 1>) {isSetParameters=true}
  ]
  particleUpdateScript [
   &ParticleState(
    Kill Particles When Lifetime Has Elapsed=false
    Let Infinitely Lived Particles Die When Emitter Deactivates=true
    Loop Particles Lifetime=true
   )
   &SkeletalMeshLocation(
    Bone Sampling Mode=::NewEnumerator4
    MaskSpawnGroup=false
    Orientation Sampling=::NewEnumerator1
    Position Sampling=::NewEnumerator2
    Skeletal Mesh=$User.SkeletalMesh1
    Use Sampled Mesh Attribute as Ribbon ID=false
   )
   &self:LIne_Location_Adv_Scratch(
    DistanceFrom Line:NiagaraFloat=&FloatFromCurve(
     FloatCurve=DataInterface() {propertyValues {Curve {keys [
         {
          arriveTangent=18.1461
          arriveTangentWeight=1.2634
          interpMode=RCIM_Cubic
          leaveTangent=18.1461
          leaveTangentWeight=1.2634
          tangentMode=RCTM_User
          tangentWeightMode=RCTWM_WeightedBoth
          time=0
          value=0
         }
         {
          arriveTangent=-15.674
          arriveTangentWeight=1.0887
          interpMode=RCIM_Cubic
          leaveTangentWeight=0.1176
          tangentMode=RCTM_Break
          tangentWeightMode=RCTWM_WeightedBoth
          time=0.25
          value=0
         }
         {
          arriveTangent=0.0112
          arriveTangentWeight=0.1547
          interpMode=RCIM_Cubic
          leaveTangent=17.6302
          leaveTangentWeight=2.6772
          tangentMode=RCTM_Break
          tangentWeightMode=RCTWM_WeightedBoth
          time=0.6
          value=0
         }
         {
          arriveTangent=-17.5526
          arriveTangentWeight=1.5902
          interpMode=RCIM_Cubic
          leaveTangent=-17.5526
          leaveTangentWeight=1.5902
          tangentMode=RCTM_User
          tangentWeightMode=RCTWM_WeightedBoth
          time=0.9027
          value=-0.0171
         }
        ]}}}
     CurveIndex=$Particles.RibbonLinkOrder
     Scale Curve=8
    )
    End Pos:Vector3f=&Add_Vector(A=Linked() {linkedVariable {name=Particles.SkeletalMeshLocation.SampledPosition, type=NiagaraPosition}}, B=<0 0 4>)
    Pos On Line:NiagaraFloat=&Lerp_Float(A=-0.1, B=1.25, Alpha=&ReturnNormalizedExecIndex())
    RotationAorundLine:NiagaraFloat=&Add_Float(
     A=&Divide_Float(
      A=&MakeFloatFromInt(Int=&SpawnGroup())
      B=&MakeFloatFromInt(Int=$Emitter.NumberOfRibbons)
     )
     B=&Multiply_Float(
      A=&Add_Float(
       A=&Multiply_Float(A=$Engine.Time, B=-0.1)
       B=&Lerp_Float(
        A=&Sine(Normalized Angle=$Emitter.Age, Period=4.6055, Scale=0.0727, Bias=-0.0364)
        B=&Cosine(Normalized Angle=$Emitter.Age, Period=8.8374, Scale=0.3136, Bias=0.0455)
        Alpha=&ReturnNormalizedExecIndex(Normalized Index Scale=0.8432)
       )
      )
      B=1
     )
    )
    Start Pos:Vector3f=$System.Orb Position
   ) {module=LIne_Location_Adv_Scratch}
   /SetVariables(Particles.NoisePosition:Vector3f=&Add_Vector(
     A=Linked() {linkedVariable {name=Particles.Position, type=NiagaraPosition}}
     B=&Multiply_Vector(
      A=/Niagara/DynamicInputs/Vectors/RotateVector(
       Vector To Rotate:Vector3f=&RandomVector(Vector Scale=&RandomRangeFloat(Minimum=0, Maximum=10))
       Yaw:NiagaraFloat=&RandomRangeFloat(Minimum=90, Maximum=-90)
       Pitch:NiagaraFloat=&RandomRangeFloat(Minimum=90, Maximum=-90)
       Roll:NiagaraFloat=&RandomRangeFloat(Minimum=-90, Maximum=90)
       Rotation Rate:NiagaraFloat=&Multiply_Float(A=$Engine.Time, B=0.8)
      )
      B=&VectorFromFloat(Value=&FloatFromCurve(
        FloatCurve=DataInterface() {propertyValues {Curve {keys [
            {interpMode=RCIM_Cubic, leaveTangentWeight=0.1333, tangentMode=RCTM_Break, tangentWeightMode=RCTWM_WeightedBoth, time=0.1, value=0}
            {
             arriveTangentWeight=0.4366
             interpMode=RCIM_Cubic
             leaveTangentWeight=0.5317
             tangentMode=RCTM_Break
             tangentWeightMode=RCTWM_WeightedBoth
             time=0.5
             value=1
            }
            {arriveTangentWeight=0.1333, interpMode=RCIM_Cubic, tangentMode=RCTM_Break, tangentWeightMode=RCTWM_WeightedBoth, time=1, value=0}
           ]}}}
        CurveIndex=&ReturnNormalizedExecIndex()
        Scale Curve=0.75
       ))
     )
    )) {isSetParameters=true}
   &ScaleRibbonWidth(Ribbon Width Scale=&Multiply_Float(
     A=&Multiply_Float(
      A=&Waveform(
       "Waveform [1]"=::NewEnumerator10
       "[1] Input Position"=&ReturnNormalizedExecIndex(Normalized Index Scale=2)
       "[1] Phase Scale"=-1
       "[1] Frequency"=1
       "[1] Amplitude Min/Max"=<0.5 1>
       "[2] Average"=true
       "Waveform [2]"=::NewEnumerator10
       "[2] Input Position"=&ReturnNormalizedExecIndex(Normalized Index Scale=3)
       "[2] Phase"=$Emitter.Age
       "Waveform [3]"=::NewEnumerator13
       "[3] Phase"=$Emitter.Age
      )
      B=&FloatFromCurve(
       FloatCurve=DataInterface() {propertyValues {Curve {keys [{interpMode=RCIM_Cubic, tangentMode=RCTM_User, time=0.25, value=0}, {interpMode=RCIM_Cubic, tangentMode=RCTM_User, time=1, value=1}]}}}
       CurveIndex=$System.FN_Timer.Value
      )
     )
     B=&FloatFromCurve(
      FloatCurve=DataInterface() {propertyValues {Curve {keys [
          {interpMode=RCIM_Cubic, leaveTangentWeight=0.0833, tangentMode=RCTM_User, time=0, value=0.5}
          {arriveTangentWeight=0.1389, interpMode=RCIM_Cubic, leaveTangentWeight=0.2778, tangentMode=RCTM_User, time=0.25, value=1}
          {arriveTangentWeight=0.2778, interpMode=RCIM_Cubic, leaveTangentWeight=0.1389, tangentMode=RCTM_User, time=0.75, value=1}
          {arriveTangentWeight=0.0833, interpMode=RCIM_Cubic, tangentMode=RCTM_User, time=1, value=0.5}
         ]}}}
      CurveIndex=$Particles.RibbonLinkOrder
      Scale Curve=0.85
     )
    ))
   &ScaleColor(
    Scale Alpha=&FloatFromCurve(
     FloatCurve=DataInterface() {propertyValues {Curve {keys [
         {
          arriveTangent=13.125
          interpMode=RCIM_Cubic
          leaveTangent=4.5187
          leaveTangentWeight=0.6328
          tangentMode=RCTM_Break
          tangentWeightMode=RCTWM_WeightedBoth
          time=0
          value=0
         }
         {
          arriveTangentWeight=0.2059
          interpMode=RCIM_Cubic
          leaveTangentWeight=0.3252
          tangentMode=RCTM_Break
          tangentWeightMode=RCTWM_WeightedBoth
          time=0.2
          value=1
         }
         {
          arriveTangent=-0.8325
          arriveTangentWeight=0.2826
          interpMode=RCIM_Cubic
          tangentMode=RCTM_Break
          tangentWeightMode=RCTWM_WeightedBoth
          time=0.9
          value=0
         }
        ]}}}
     CurveIndex=&ReturnNormalizedExecIndex()
    )
    ScaleA=true
    ScaleRGB=true
   )
  ]
  rendererClasses [&NiagaraRibbonRendererProperties]
  renderers [{rendererClass=&NiagaraRibbonRendererProperties}]
 }
 Orb {
  emitterUpdateScript [
   &EmitterState()
   &SpawnBurst_Instantaneous(Age:NiagaraFloat=$Emitter.Age, Spawn Count=3, Use Spawn Probability=false)
  ]
  particleSpawnScript [
   &InitializeParticle(
    Color=&HueShiftLinearColor(
     Hue=&MakeFloatFromVector(Vector=$User.HSV Shift, Channel=::NewEnumerator0)
     Saturation=&MakeFloatFromVector(Vector=$User.HSV Shift, Channel=::NewEnumerator1)
     Value=&MakeFloatFromVector(Vector=$User.HSV Shift, Channel=::NewEnumerator2)
     Linear Color=<rgba 0 25 3.3333 0.25>
    )
    Lifetime=5
    Mesh Scale Mode=::NewEnumerator3
    Mesh Uniform Scale=0.25
    Sprite Size Max=<11 14>
    Sprite Size Min=<8 12>
    Sprite Size Mode=::NewEnumerator2
    UsePositionOffset=false
    Write Material Random=true
   )
   &InitialMeshOrientation(Orientation Axis=<0 0 1>, Use Orientation Vector=true, Use Rotation Vector=false)
   &CameraOffset(Camera Offset Amount=11.726)
   /SetVariables() {isSetParameters=true}
  ]
  particleUpdateScript [
   &ParticleState(
    Kill Particles When Lifetime Has Elapsed=false
    Let Infinitely Lived Particles Die When Emitter Deactivates=true
    Loop Particles Lifetime=true
   )
   /SetVariables(Particles.Position:Vector3f=$System.Orb Position) {isSetParameters=true, module=SetVariables_#4}
   &ScaleSpriteSize(Scale Factor=&Vector2DFromFloat(Value=&FloatFromCurve(
      FloatCurve=DataInterface() {propertyValues {Curve {keys [
          {
           arriveTangent=-0.237
           arriveTangentWeight=0.6278
           interpMode=RCIM_Cubic
           leaveTangent=-0.237
           leaveTangentWeight=0.6278
           tangentMode=RCTM_User
           tangentWeightMode=RCTWM_WeightedBoth
           time=0
           value=0
          }
          {
           arriveTangent=-5.3547
           arriveTangentWeight=2.2634
           interpMode=RCIM_Cubic
           leaveTangent=-5.3547
           leaveTangentWeight=2.2634
           tangentMode=RCTM_User
           tangentWeightMode=RCTWM_WeightedBoth
           time=1
           value=1
          }
         ]}}}
      CurveIndex=$System.FN_Timer.Value
     )))
   &SpriteRotationRate(Rotation Rate=&RandomRangeFloat(Minimum=180, Maximum=-180))
   &ScaleMeshSize(Scale Factor=&VectorFromFloat(Value=&FloatFromCurve(
      FloatCurve=DataInterface() {propertyValues {Curve {keys [
          {
           arriveTangent=-0.237
           arriveTangentWeight=0.6278
           interpMode=RCIM_Cubic
           leaveTangent=-0.237
           leaveTangentWeight=0.6278
           tangentMode=RCTM_User
           tangentWeightMode=RCTWM_WeightedBoth
           time=0
           value=0
          }
          {
           arriveTangent=-5.3547
           arriveTangentWeight=2.2634
           interpMode=RCIM_Cubic
           leaveTangent=-5.3547
           leaveTangentWeight=2.2634
           tangentMode=RCTM_User
           tangentWeightMode=RCTWM_WeightedBoth
           time=1
           value=1
          }
         ]}}}
      CurveIndex=$System.FN_Timer.Value
     )))
  ]
  rendererClasses [&NiagaraSpriteRendererProperties]
  renderers [{rendererClass=&NiagaraSpriteRendererProperties}]
 }
 OrbLight {
  emitterUpdateScript [
   &EmitterState()
   &SpawnBurst_Instantaneous(Age:NiagaraFloat=$Emitter.Age, Spawn Count=1, Use Spawn Probability=false)
  ]
  particleSpawnScript [&InitializeParticle(
    Color=&HueShiftLinearColor(
     Hue=&MakeFloatFromVector(Vector=$User.HSV Shift, Channel=::NewEnumerator0)
     Saturation=&MakeFloatFromVector(Vector=$User.HSV Shift, Channel=::NewEnumerator1)
     Value=&MakeFloatFromVector(Vector=$User.HSV Shift, Channel=::NewEnumerator2)
     Linear Color=<rgba 0 25 3.3333 0.25>
    )
    Lifetime=5
    Mesh Scale Mode=::NewEnumerator3
    Mesh Uniform Scale=0.25
    Sprite Size Max=<11 14>
    Sprite Size Min=<8 12>
    Sprite Size Mode=::NewEnumerator2
    UsePositionOffset=false
    Write Material Random=true
   )]
  particleUpdateScript [
   &ParticleState(
    Kill Particles When Lifetime Has Elapsed=false
    Let Infinitely Lived Particles Die When Emitter Deactivates=true
    Loop Particles Lifetime=true
   )
   /SetVariables(Particles.Position:Vector3f=$System.Orb Position) {isSetParameters=true, module=SetVariables_#4}
   &FN_Light_Attributes(
    Color=&HueShiftLinearColor(
     Hue=&MakeFloatFromVector(Vector=$User.HSV Shift, Channel=::NewEnumerator0)
     Saturation=&MakeFloatFromVector(Vector=$User.HSV Shift, Channel=::NewEnumerator1)
     Value=&MakeFloatFromVector(Vector=$User.HSV Shift, Channel=::NewEnumerator2)
     Linear Color=<rgba 0 1 0.1333 1>
    )
    Create Color Attribute=true
    Create Exponent Attribute=true
    Create Radius Attribute=true
    Enable Particle Light Enable=true
    Exponent=3
    Light Enabled=true
    Radius=50
    Scale Color=&MakeCustomFloatFromBool(Bool=$User.bIsMindcontrolling, True Float=1, False Float=0)
   )
  ]
  rendererClasses [&NiagaraLightRendererProperties]
  renderers [{rendererClass=&NiagaraLightRendererProperties, rendererIndex=0}]
 }
}
summary {
 systemName=NS_Invader_Pet_MindControl
 emitters [
  {emitterName=Orb, rendererClasses [&NiagaraSpriteRendererProperties]}
  {emitterName=Hand_Rings, rendererClasses [&NiagaraMeshRendererProperties]}
  {emitterName=Lines_Orb_To_Head, rendererClasses [&NiagaraRibbonRendererProperties]}
  {emitterName=OrbLight, rendererClasses [&NiagaraLightRendererProperties]}
 ]
 userVariables [
  User.bIsMindcontrolling {type=NiagaraBool}
  User.HSV Shift {type=Vector3f, description="", defaultValue {struct=Vector3f, value=<0 1 1>}}
  User.SkeletalMesh {
   type=NiagaraDataInterfaceSkeletalMesh
   defaultValue {
    struct=NiagaraExt_VariableValue_DataInterface
    value {dataInterface=&self:NiagaraDataInterfaceSkeletalMesh_2}
   }
  }
  User.SkeletalMesh1 {
   type=NiagaraDataInterfaceSkeletalMesh
   defaultValue {
    struct=NiagaraExt_VariableValue_DataInterface
    value {dataInterface=&self:NiagaraDataInterfaceSkeletalMesh_0}
   }
  }
 ]
}
